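# Material function that builds base color and glowing emissive for the ProdigyFire pickaxe blade.
# The FX texture's red mask darkens base color, restored where inverted EffectMask applies.
# A sine-over-time pulse scales emissive from EmissiveColor and E-texture-masked color, faded by TransitionAmountForGlow.
# A masked red glow is added before GlowIntensity scaling; time of day is a constant, leaving the Scalar_Time_of_Day call unconnected.
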MaterialFunction /Game/Weapons/FORT_Melee/Pickaxe_ProdigyFire/Materials/MaterialFunctions/MF_Pickaxe_ProdigyFire_BladeGlow
graph {nodes [
  n1 = Time()
  n2 = Multiply(A=Sine(None=Multiply(A=@n1) {constB=0.5})) {constB=0.5}
  n3 = FunctionInput() {bUsePreviewValueAsDefault=true, inputName=BaseColor}
  n4 = TextureSample() {samplerType=SAMPLERTYPE_Masks, texture=/Game/Weapons/FORT_Melee/Pickaxe_ProdigyFire/Textures/T_Pickaxe_ProdigyFire_FX}
  n5 = LinearInterpolate(A=@n3, B=Constant3Vector() {constant=<rgba 0.0222 0.0222 0.0222 1>}, Alpha=@n4.R)
  n6 = Multiply(A=@n5)
  n7 = TextureSample() {texture=/Game/Weapons/FORT_Melee/Pickaxe_ProdigyFire/Textures/T_Pickaxe_ProdigyFire_E}
  n8 = Multiply(
   A=LinearInterpolate(Alpha=Add(A=@n2) {constB=0.5}) {constA=0.6, constB=2.5}
   B=Add(A=FunctionInput() {bUsePreviewValueAsDefault=true, inputName=EmissiveColor, sortPriority=1}, B=LinearInterpolate(B=@n6, Alpha=@n7.R))
  )
  n9 = Constant() {desc=TimeOfDay, r=1}
  n10 = Multiply(A=@n4.R, B=OneMinus(None=FunctionInput() {inputName=EffectMask, sortPriority=3}))
  n11 = FunctionInput() {bUsePreviewValueAsDefault=true, inputName=RedGlowColor, sortPriority=2, previewValue=<0.677 0.0544 0.0544 1>}
  n12 = Multiply(A=@n11, B=FunctionInput() {bUsePreviewValueAsDefault=true, inputName=RedGlowMultiplier, inputType=FunctionInput_Scalar, sortPriority=2})
  n13 = Add(
   A=LinearInterpolate(B=Multiply(A=@n8, B=@n9), Alpha=FunctionInput() {bUsePreviewValueAsDefault=true, inputName=TransitionAmountForGlow, sortPriority=6})
   B=Multiply(A=@n10, B=Multiply(A=@n12, B=@n9))
  )
  Add(A=@n13, B=Multiply(A=@n4.R) {constB=3})
  n14 = FunctionInput() {bUsePreviewValueAsDefault=true, inputName=GlowIntensity, inputType=FunctionInput_Scalar, previewValue=<1 0 0 1>}
  FunctionOutput(None=Multiply(A=@n13, B=@n14)) {outputName=Emissive Color Out, sortPriority=1}
  FunctionOutput(None=LinearInterpolate(A=@n5, B=@n3, Alpha=@n10)) {outputName=Base Color Out}
  n15 = Constant() {r=1}
  MaterialFunctionCall(Value_Morning=@n15, Value_Day=@n15, Value_Evening=Constant() {r=0.8}, Value_Night=Constant() {r=0.5}) {materialFunction=&Scalar_Time_of_Day}
 ]}
