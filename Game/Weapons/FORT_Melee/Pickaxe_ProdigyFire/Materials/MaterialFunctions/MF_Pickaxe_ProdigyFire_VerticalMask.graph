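# Material function for the ProdigyFire pickaxe that outputs a saturated vertical gradient mask driven by TransitionAmount.
# It transforms world position out of world space, scales it by MaskHardness, offsets it between MinPoint and MaxPoint by TransitionAmount, and keeps the Z channel.
# An optional static-switched initial fade blends from full white into that mask as TransitionAmount rises to InitialFadeThreshold.
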MaterialFunction /Game/Weapons/FORT_Melee/Pickaxe_ProdigyFire/Materials/MaterialFunctions/MF_Pickaxe_ProdigyFire_VerticalMask
graph {nodes [
  n1 = TransformPosition(None=WorldPosition().XYZ) {transformSourceType=TRANSFORMPOSSOURCE_World}
  n2 = FunctionInput() {bUsePreviewValueAsDefault=true, inputName=MaskHardness, inputType=FunctionInput_Scalar, previewValue=<1 0 0 1>}
  n3 = Multiply(A=@n2) {constB=0.05}
  n4 = FunctionInput() {bUsePreviewValueAsDefault=true, inputName=MinPoint, inputType=FunctionInput_Scalar, previewValue=<-10 0 0 1>}
  n5 = Multiply(A=@n4, B=@n3) {constA=-10}
  n6 = FunctionInput() {bUsePreviewValueAsDefault=true, inputName=MaxPoint, inputType=FunctionInput_Scalar, previewValue=<-80 0 0 1>}
  n7 = Multiply(A=@n3, B=@n6) {constA=-80, constB=-80}
  n8 = FunctionInput() {bUsePreviewValueAsDefault=true, inputName=TransitionAmount, inputType=FunctionInput_Scalar}
  n9 = ComponentMask(None=Add(A=Multiply(A=@n1, B=@n3), B=LinearInterpolate(A=@n5, B=@n7, Alpha=@n8) {constA=-15, constB=-80}) {constB=-80}) {b=true}
  n10 = FunctionInput() {bUsePreviewValueAsDefault=true, inputName=InitialFadeThreshold, inputType=FunctionInput_Scalar, previewValue=<0.2 0 0 1>}
  n11 = LinearInterpolate(B=@n9, Alpha=Saturate(None=InverseLinearInterpolate(B=@n10, Value=@n8) {constB=0.2})) {constA=1}
  n12 = FunctionInput(Preview=StaticBool()) {bUsePreviewValueAsDefault=true, inputName=UseInitialFade, inputType=FunctionInput_StaticBool}
  FunctionOutput(None=Saturate(None=StaticSwitch(True=@n11, False=@n9, Value=@n12))) {outputName=VerticalMask}
 ]}
