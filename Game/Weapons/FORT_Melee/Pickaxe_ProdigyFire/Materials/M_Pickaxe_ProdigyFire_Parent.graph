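# Parent material for the ProdigyFire pickaxe, layering a red "hit transition" blade glow over the standard 2023 character shader.
# A vertical mask function, softened or hardened by transition progress, feeds the blade glow function to alter base color and emissive.
# ReactivityStatus, compared against 1, switches transition and glow amounts between parameter-driven, fixed and zero values.
# Emissive passes through cosmetic eye-adaptation compensation before being written back to the material attributes.
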Material /Game/Weapons/FORT_Melee/Pickaxe_ProdigyFire/Materials/M_Pickaxe_ProdigyFire_Parent
graph {
 properties {
  bUseMaterialAttributes=true
  bUsedWithInstancedStaticMeshes=true
  bUsedWithMeshDeformer=true
  bUsedWithNanite=true
  bUsedWithSkeletalMesh=true
  pixelDepthOffsetMode=PDOM_Legacy
  usedShadingModels=Default Lit
 }
 nodes [
  n1 = MaterialFunctionCall() {materialFunction=&MF_CharacterParent_2023_PreFX}
  n2 = GetMaterialAttributes(None=MaterialFunctionCall(Material=@n1.Result, M=@n1.Result, Pre Skinned Local Position=@n1.Result, Distance Blend=@n1.Result, Fresnel=@n1.Result) {materialFunction=&MF_CharacterParent_2023_PostFX}.Result) {attributeGetTypes [#1, #2]}
  n3 = ScalarParameter() {group=RedHitTransition, parameterName=ReactivityStatus, sortPriority=20}
  n4 = ScalarParameter() {defaultValue=1, group=RedHitTransition, parameterName=TransitionHardness, sliderMax=3, sliderMin=0.01}
  n5 = Saturate(None=InverseLinearInterpolate(Value=ScalarParameter() {group=RedHitTransition, parameterName=TransitionAmount, sliderMax=1}) {constB=0.5})
  n6 = Constant()
  n7 = Constant() {r=1}
  n8 = ScalarParameter() {group=RedHitTransition, parameterName=TransitionAmount, sliderMax=1}
  n9 = MaterialFunctionCall!(
   GlowIntensity=ScalarParameter() {defaultValue=1, group=RedHitTransition, parameterName=MasterGlowIntensity, sliderMax=1}
   BaseColor=@n2.Base Color
   EmissiveColor=@n2.Base Color
   RedGlowColor=VectorParameter() {group=RedHitTransition, parameterName=RedGlowColor, defaultValue=<rgba 0.6771 0.0544 0.0544 1>}.RGB
   RedGlowMultiplier=If(
    A=@n3
    "A > B"=Constant()
    "A == B"=Constant() {r=0.1}
    "A < B"=ScalarParameter() {group=RedHitTransition, parameterName=RedGlowMultiplier, sliderMax=1}
   ) {constB=1}
   EffectMask=MaterialFunctionCall!(
    MaskHardness=LinearInterpolate(B=@n4, Alpha=@n5) {constA=3}
    MinPoint=ScalarParameter() {defaultValue=-10, group=RedHitTransition, parameterName=MaskMin, sliderMax=500, sliderMin=-15}
    MaxPoint
    InitialFadeThreshold=ScalarParameter() {group=RedHitTransition, parameterName=InitialFadeThreshold}
    UseInitialFade=StaticBool() {value=true}
    TransitionAmount=If(A=@n3, "A > B"=@n6, "A == B"=@n7, "A < B"=@n8) {constB=1}
   ) {materialFunction=/Game/Weapons/FORT_Melee/Pickaxe_ProdigyFire/Materials/MaterialFunctions/MF_Pickaxe_ProdigyFire_VerticalMask}.VerticalMask
   TransitionAmountForGlow=If(
    A=@n3
    "A > B"=Constant()
    "A == B"=Constant() {r=1}
    "A < B"=ScalarParameter() {group=RedHitTransition, parameterName=TransitionAmountForGlow, sliderMax=1, sortPriority=20}
   ) {constB=1}
  ) {materialFunction=/Game/Weapons/FORT_Melee/Pickaxe_ProdigyFire/Materials/MaterialFunctions/MF_Pickaxe_ProdigyFire_BladeGlow}
  n10 = MaterialFunctionCall!(Emissive=@n9.Emissive Color Out, Alpha=Constant() {r=1}, Material=@n2.MaterialAttributes) {materialFunction=/Game/Packages/Fortress_SharedMaterials/Base_Material_Functions/BaseCharacterShaderFunctions/MF_CosmeticEmissiveEyeAdaptation}
  SetMaterialAttributes!(MaterialAttributes=@n2.MaterialAttributes, Base Color=@n9.Base Color Out, Emissive Color=@n10.Emissive) {attributeSetTypes [#1, #2]}
 ]
}
